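; Front-end Spy Base set-piece for Donut Week 7 that presents an overflowing toilet with a mesh, particles and looping-style overflow audio.
; The construction script feeds the Splash Offset variable into the particle's floor splash Z offset and sets the overflow mesh's Z scale from a variable.
; On BeginPlay the audio stops if the actor is hidden and fades in otherwise, then fades out when the actor is destroyed.
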(event EventDestroyed
  (bind _overflow_audio (Variables|Default|GetOverflow_Audio)
    (:then
      (Audio|Components|Audio|FadeOut _overflow_audio 0.25 0.0 "Linear" :!class /Script/Engine.AudioComponent))
    (:else)))

(event EventBeginPlay
  (if (Variables|Rendering|GetbHidden)
    (Audio|Components|Audio|Stop (Variables|Default|GetOverflow_Audio))
    (else
      (Audio|Components|Audio|FadeIn (Variables|Default|GetOverflow_Audio) 0.25 1.0 0.0 "Linear" :!class /Script/Engine.AudioComponent))))

(event Collision|EventActorBeginOverlap
  (params (OtherActor :type Actor%20Object%20Reference)))

(event EventTick
  (params (DeltaSeconds :type Float%20%28single-precision%29)))

(fn ConstructionScript
  (Effects|Components|ParticleSystem|SetVectorParameter (Variables|Default|GetP_Toilet_Overflow) "Floor_Splash_Offset" (Math|Vector|MakeVector 0.0 0.0 (Variables|Default|GetSplash%20Offset)))
  (Transformation|SetWorldScale3D (Variables|Default|Getsm_spybase_donut_toilet_overflow) (Math|Vector|MakeVector 1.063016 1.063016 (Variables|Default|GetOverflowMeshZScale))))
